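# Preview-only unlit material for a melee slash trail mesh with tweakable thickness, building an offset vector along vertex normals.
# A V gradient on UV channel 1, shaped by a power curve, weights the normal, which the Thickness parameter scales.
# DebugTimeSine animates the offset from zero, and UseScaleWidthByObjectSize blends in object X scale.
# No material output connection is shown, and the DynamicParameter node is unconnected.
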Material /Game/Effects/Fort_Effects/Mesh/Melee/ThicknessTweakableSlashMesh/M_MeleeTrail_Slash_ThicknessTweakable_PreviewOnly
graph {
 properties {
  bAutomaticallySetUsageInEditor=false
  previewMesh=/Engine/Transient.MaterialEditorMeshComponent_2
  shadingModel=MSM_Unlit
  usedShadingModels=Unlit
 }
 nodes [
  DynamicParameter() {paramNames [Thickness,Param2,Param3,Param4]}
  LinearInterpolate() {constA=1}
  MaterialFunctionCall() {materialFunction=&LinearGradient}
  n1 = MaterialFunctionCall(UV Channel=TextureCoordinate() {coordinateIndex=1}) {materialFunction=&LinearGradient}
  n2 = Constant() {r=1.1}
  n3 = Multiply(A=VertexNormalWS(), B=LinearInterpolate(Alpha=Power(Base=@n1.VGradient, Exp=@n2)))
  n4 = ScalarParameter() {defaultValue=71, parameterName=Thickness}
  Multiply(
   A=LinearInterpolate(
    B=MaterialFunctionCall() {materialFunction=&ObjectScale}.Scale X
    Alpha=ScalarParameter() {defaultValue=1, parameterName=UseScaleWidthByObjectSize}
   ) {constA=1}
   B=LinearInterpolate(
    B=Multiply(A=@n3, B=@n4)
    Alpha=MaterialFunctionCall() {materialFunction=&DebugTimeSine}.Result
   )
  )
 ]
}
